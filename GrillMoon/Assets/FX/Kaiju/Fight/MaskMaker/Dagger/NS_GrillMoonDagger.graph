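# Niagara system for the charge glow on the GrillMoon dagger in the MaskMaker kaiju fight.
# It layers a rotating corona mesh, spiralling airflow meshes, a mesh glow, pulsing blue-purple glow sprites and GPU sparkles.
# System.DaggerCharge drives the opacity of every layer except the sparkles, with curves shaping both glow responses.
# User.UseCombinedFX sets sparkle spawn probability and blends in a corona size pulse, which User.CoronaFlareScale multiplies.
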NiagaraSystem /GrillMoon/Assets/FX/Kaiju/Fight/MaskMaker/Dagger/NS_GrillMoonDagger
dependencies {
 usedDataInterfaces [
  &NiagaraDataInterfaceCurve
  &NiagaraDataInterfaceColorCurve
  &NiagaraDataInterfaceVectorCurve
  &NiagaraDataInterfaceVector2DCurve
 ]
 usedDynamicInputs []
 usedModules [
  &self:SystemSpawnScript.SystemScriptSource.SystemScriptGraph.NiagaraNodeAssignment_1.SetVariables_#1
  &SystemState
  &self:SystemSpawnScript.SystemScriptSource.SystemScriptGraph.NiagaraNodeAssignment_0.SetVariables_#2
 ]
}
emitters {
 RainbowCorona003 {
  emitterUpdateScript [
   &EmitterState(
    Inactive Response=::NewEnumerator1
    Life Cycle Mode=::NewEnumerator1
    Loop Behavior=::NewEnumerator1
    Loop Duration Mode=::NewEnumerator1
    UseLoopDelay=false
   )
   &SpawnBurst_Instantaneous(Use Spawn Probability=false, UseLoopCountLimit=false)
  ]
  particleSpawnScript [
   &InitializeParticle(
    Color=<rgba 0.25 1 1 0.2>
    Lifetime=5
    Mesh Scale Mode=::NewEnumerator3
    Mesh Uniform Scale=0.36
    Position Offset=&Multiply_VectorByFloat(Vector=<0.15 0 0>, Float=50)
    UsePositionOffset=true
    Write Material Random=true
   )
   &DynamicMaterialParameters(
    Index 0 Param 1=&Multiply_Float(A=3178.1782, B=1.5)
    Module.Param0WriteEnabled=true
    Module.Param0WriteEnabled1=true
    Module.Param1WriteEnabled=true
    Module.Param1WriteEnabled1=true
    Module.Param2WriteEnabled=true
    Module.Param2WriteEnabled1=true
    Module.Param3WriteEnabled=true
    Module.Param3WriteEnabled1=true
    Write Parameter Index 1=true
   )
   &InitialMeshOrientation(
    Mesh Axis To Orient=<0 0 1>
    Mesh Orientation Mode=::NewEnumerator2
    Rotation=&RandomRangeVector(Minimum=<0 0 0>, Maximum=<0 0 1>)
    Use Rotation Vector=true
   )
  ]
  particleUpdateScript [
   &ParticleState(Kill Particles When Lifetime Has Elapsed=false, Let Infinitely Lived Particles Die When Emitter Deactivates=true)
   &ScaleMeshSize(Scale Factor=&VectorFromFloat(Value=100)) {module=ScaleMeshSize001}
   &UpdateMeshOrientation(Rotation Rate=0.005, Rotation Vector=<0 0 1>)
   &ScaleMeshSize(Scale Factor=&VectorFromFloat(Value=&Multiply_Float(
      A=$User.CoronaFlareScale
      B=&Lerp_Float(
       A=1
       B=&Waveform("Waveform [1]"=::NewEnumerator12, "[1] Phase Scale"=0.5, "[1] Frequency"=2, "[1] Amplitude Min/Max"=<0.8 1>)
       Alpha=$User.UseCombinedFX
      )
     )))
   &ScaleColor(Scale Alpha=1.5, Scale RGB=&VectorFromFloat(Value=10), ScaleA=true, ScaleRGB=true)
   &ScaleColor(Scale Alpha=$System.DaggerCharge, Scale RGB=&VectorFromFloat(Value=1.5), ScaleA=true, ScaleRGB=true) {module=ScaleColor001}
   /SetVariables() {isSetParameters=true}
  ]
  rendererClasses [&NiagaraMeshRendererProperties]
  renderers [{rendererClass=&NiagaraMeshRendererProperties, rendererIndex=0}]
 }
 Sparklies {
  simTarget=GPUComputeSim
  emitterUpdateScript [
   &EmitterState()
   &SpawnRate(
    Spawn Probability=$User.UseCombinedFX
    SpawnRate=&Lerp_Float(A=300, B=75, Alpha=&Sine(Normalized Angle=$Emitter.Age, Period=2, Scale=0.5, Bias=0.5))
    Use Spawn Probability=true
   )
   &SpawnBurst_Instantaneous(Spawn Count=100, Use Spawn Probability=false) {enabled=false}
  ]
  particleSpawnScript [
   /Niagara/Modules/Spawn/Initialization/InitializeParticle(
    Color=<rgba 0.0852 0.5416 1 1>
    Lifetime=&UniformRangedFloat(Minimum=0.6, Maximum=0.8)
    Module.Write Color=true
    Module.Write Lifetime=true
    Module.Write Mass=true
    Module.Write Material Random=true
    Module.Write Position=false
    Module.Write Scale=false
    Module.Write SpriteRotation=true
    Module.Write SpriteSize=true
    Position=&SimulationPosition()
    Sprite Rotation=&AngleConversion(Angle:NiagaraFloat=0, Angle Input:ENiagara_AngleInput=::NewEnumerator0, Angle Output:ENiagara_AngleInput=::NewEnumerator1)
    Sprite Size=&Vector2DFromFloat(Value=&FloatFromCurve(FloatCurve=DataInterface() {propertyValues {Curve {keys [<at 0 80>,<at 1 50>]}}}, CurveIndex=$Emitter.NormalizedLoopAge))
   ) {enabled=false}
   /Niagara/Modules/Spawn/Initialization/V2/InitializeParticle(
    Color=<rgba 0.0852 0.5416 1 1>
    Lifetime Max=0.75
    Lifetime Min=0.2
    Lifetime Mode=::NewEnumerator1
    Mass Mode=::NewEnumerator1
    Sprite Rotation Angle=&AngleConversion(Angle:NiagaraFloat=0, Angle Input:ENiagara_AngleInput=::NewEnumerator0, Angle Output:ENiagara_AngleInput=::NewEnumerator1)
    Sprite Rotation Mode=::NewEnumerator0
    Sprite Size Max=<25 40>
    Sprite Size Min=<15 55>
    Sprite Size Mode=::NewEnumerator2
    UsePositionOffset=false
    Write Material Random=true
   ) {module=InitializeParticle001}
   &AddVelocity(Velocity=&Multiply_Vector(
     A=$Output.SphereLocation.SphereNormal
     B=&VectorFromFloat(Value=&UniformRangedFloat(Minimum=1, Maximum=5))
    ))
   &ShapeLocation(
    Box Midpoint=<0.25 0.5 0.5>
    Box Size=<40 5 10>
    MaskSpawnGroup=false
    Offset Mode=::NewEnumerator0
    Override Seed=false
    Shape Primitive=::NewEnumerator2
    Surface Only=false
    UseOwnerScale=true
   )
   &AddVelocity(
    Velocity=&RandomVector(Vector Scale=1)
    Velocity Speed Scale=&Multiply_Float(A=&RandomRangeFloat(Minimum=25, Maximum=500), B=$System.FX Scale)
   ) {module=AddVelocity001}
   &InheritVelocity(
    Inherited Velocity Amount Scale=&VectorFromFloat(Value=0.1)
    Speed Limit=false
    UseSpeedThreshold=false
   )
  ]
  particleUpdateScript [
   &ParticleState()
   &AccelerationForce(Acceleration=&Multiply_Vector(
     A=<0 0 -980>
     B=&VectorFromFloat(Value=&Multiply_Float(A=0.3, B=$System.FX Scale))
    ))
   &VectorNoiseForce(
    Force Amount=&Multiply_Float(A=$System.FX Scale, B=&UniformRangedFloat(Minimum=1000, Maximum=2500))
    Module.Use Random Noise Time Interval=true
    Random Noise Time Interval=0.05
   )
   &Drag(Drag=&UniformRangedFloat(Minimum=2, Maximum=3), Use Linear Drag=true, Use Rotational Drag=false)
   &SolveForcesAndVelocity(Clamp Velocity=false, Limit Acceleration=false)
   &ScaleColor(
    Scale Alpha=&FloatFromCurve(FloatCurve=DataInterface() {propertyValues {Curve {keys [<at 0.7 1 RCIM_Cubic>,<at 1 0>]}}})
    Scale RGB=&VectorFromCurve(VectorCurve=DataInterface() {propertyValues {
       XCurve {keys [<at 0.1 2 RCIM_Cubic>,<at 0.6981 0.2997 RCIM_Cubic>]}
       YCurve {keys [<at 0.1 2 RCIM_Cubic>,<at 0.6906 1.6013 RCIM_Cubic>]}
       ZCurve {keys [<at 0.1 2 RCIM_Cubic>,<at 0.7 1 RCIM_Cubic>]}
      }})
    ScaleA=true
    ScaleRGB=true
   )
   &SpriteRotationRate(Rotation Rate=&UniformRangedFloat(Minimum=-360, Maximum=360))
   &ScaleSpriteSize(Scale Factor=&Vector2DFromFloat(Value=&FloatFromCurve(
      FloatCurve=DataInterface() {propertyValues {Curve {keys [
          <at 0 0.49 RCIM_Cubic>
          {interpMode=RCIM_Cubic, tangentMode=RCTM_User, time=0.157, value=0.526}
          {arriveTangent=-5.4783, interpMode=RCIM_Cubic, leaveTangent=-5.4783, time=0.165, value=1.5}
          {interpMode=RCIM_Cubic, tangentMode=RCTM_User, time=0.18, value=0.4}
          {interpMode=RCIM_Cubic, tangentMode=RCTM_User, time=0.38, value=0.26}
          {arriveTangent=-0.2, interpMode=RCIM_Cubic, leaveTangent=-0.2, time=0.41, value=0.83}
          {interpMode=RCIM_Cubic, tangentMode=RCTM_User, time=0.43, value=0.25}
          {interpMode=RCIM_Cubic, tangentMode=RCTM_User, time=0.689, value=0.24}
          {arriveTangent=-0.7407, interpMode=RCIM_Cubic, leaveTangent=-0.7407, time=0.705, value=1}
          {interpMode=RCIM_Cubic, tangentMode=RCTM_User, time=0.716, value=0.22}
          <at 1 0 RCIM_Cubic>
         ]}}}
      Scale Curve=&Multiply_Float(A=1.25, B=$System.FX Scale)
     ))) {enabled=false}
   &ScaleSpriteSize(Scale Factor=&Vector2DFromFloat(Value=&FloatFromCurve(
      FloatCurve=DataInterface() {propertyValues {Curve {keys [<at 0 1 RCIM_Cubic>, {arriveTangent=-1.5251, interpMode=RCIM_Cubic, leaveTangent=-1.5251, tangentMode=RCTM_User, time=1, value=0}]}}}
      Scale Curve=&Multiply_Float(A=2, B=$System.FX Scale)
     ))) {module=ScaleSpriteSize002}
   &ScaleSpriteSize(
    Non-Uniform Curve Sprite Scale=DataInterface() {propertyValues {
      XCurve {keys [{tangentMode=RCTM_User, time=0, value=1}, {tangentMode=RCTM_User, time=1, value=1}]}
      YCurve {keys [{tangentMode=RCTM_User, time=0, value=1}, {tangentMode=RCTM_User, time=0.25, value=0.5}]}
     }}
    Scale Sprite Size Mode=::NewEnumerator3
   ) {module=ScaleSpriteSize001}
  ]
  rendererClasses [&NiagaraSpriteRendererProperties]
  renderers [{rendererClass=&NiagaraSpriteRendererProperties}]
 }
 Spiral_AirFlow {
  emitterUpdateScript [
   &EmitterState(
    Inactive Response=::NewEnumerator1
    Life Cycle Mode=::NewEnumerator1
    Loop Behavior=::NewEnumerator1
    Loop Delay=1
    Loop Duration Mode=::NewEnumerator1
    UseLoopDelay=false
   )
   &SpawnBurst_Instantaneous(Use Spawn Probability=false, UseLoopCountLimit=false)
   &SpawnRate(Use Spawn Probability=true)
  ]
  particleSpawnScript [
   &InitializeParticle(
    Color Maximum=<rgba 0.1 0.1704 1 1>
    Color Minimum=<rgba 0.1 0.6502 1 1>
    Color Mode=::NewEnumerator2
    Lifetime Min=3
    Lifetime Mode=::NewEnumerator1
    Mesh Scale Mode=::NewEnumerator3
    Mesh Uniform Scale=&Multiply_Float(A=&RandomRangeFloat(Minimum=0.7, Maximum=1), B=0.5)
    Position Offset=&RandomRangeVector(Minimum=<-15 0 0>, Maximum=<-10 0 0>)
    UsePositionOffset=true
    Write Material Random=true
   )
   &CameraOffset(Camera Offset Amount=10)
   &InitialMeshOrientation(Mesh Orientation Mode=::NewEnumerator2, Rotation=<0 0.25 0>, Use Rotation Vector=true)
   &InitialMeshOrientation(
    Mesh Orientation Mode=::NewEnumerator2
    Orientation Vector=<0 0 -1>
    Rotation=&RandomRangeVector(Minimum=<0 0 0>, Maximum=<0 0 1>)
    Use Rotation Vector=true
   ) {module=InitialMeshOrientation001}
  ]
  particleUpdateScript [
   &Drag(Drag=2, Use Linear Drag=true, Use Rotational Drag=true)
   &SolveForcesAndVelocity(Clamp Velocity=false, Limit Acceleration=false)
   &ParticleState()
   &UpdateMeshOrientation(Rotation Rate=&RandomRangeFloat(Minimum=0.5, Maximum=0.7))
   &ScaleMeshSize(Scale Factor=&Multiply_VectorByFloat(
     Vector=&Multiply_Vector(A=<1.45 1 1>, B=<1.2 2 2>)
     Float=&Multiply_Float(A=0.7, B=1)
    ))
   &ScaleColor(
    Scale Alpha=&FloatFromCurve(
     FloatCurve=DataInterface() {propertyValues {Curve {keys [
         {interpMode=RCIM_Cubic, tangentMode=RCTM_User, time=0, value=0}
         {interpMode=RCIM_Cubic, tangentMode=RCTM_User, time=0.5, value=0.8}
         {tangentMode=RCTM_User, time=1, value=0}
        ]}}}
     Scale Curve=$System.DaggerCharge
    )
    Scale RGB=&Multiply_VectorByFloat(
     Vector=&MakeVectorFromLinearColorRGB(Color=&ColorFromCurve(ColorCurve=DataInterface() {propertyValues {
         AlphaCurve {keys []}
         BlueCurve {keys [<at 0 1.3438 RCIM_Cubic>,<at 0.8254 0.5 RCIM_Cubic>]}
         GreenCurve {keys [<at 0 1.9453 RCIM_Cubic>,<at 0.8254 0.5 RCIM_Cubic>]}
         RedCurve {keys [<at 0 2 RCIM_Cubic>,<at 0.8254 0.5 RCIM_Cubic>]}
        }}))
     Float=&RandomRangeFloat(Minimum=4, Maximum=7)
    )
    ScaleA=true
    ScaleRGB=true
   )
   &DynamicMaterialParameters(
    Index 0 Param 3=1
    Index 0 Param 4=1
    Module.Index0Param0Write=true
    Module.Index0Param1Write=true
    Module.Index0Param2Write=true
    Module.Index0Param3Write=true
   )
  ]
  rendererClasses [&NiagaraMeshRendererProperties]
  renderers [{rendererClass=&NiagaraMeshRendererProperties, rendererIndex=0}]
 }
 ble_putrple_glow {
  emitterUpdateScript [
   &EmitterState()
   &SpawnRate(SpawnRate=0.7, Use Spawn Probability=false)
   &SpawnBurst_Instantaneous(Spawn Count=2, Use Spawn Probability=false, UseLoopCountLimit=true)
  ]
  particleSpawnScript [
   &InitializeParticle(
    Lifetime=2
    Position Offset=&Multiply_VectorByFloat(Vector=<0.5 0 0>, Float=10)
    Sprite Size=<3000 7000>
    Sprite Size Mode=::NewEnumerator4
    Sprite UV Mode=::NewEnumerator3
    UsePositionOffset=true
    Write Material Random=true
   )
   &DynamicMaterialParameters(
    Index 0 Param 1=0.5
    Module.Param0WriteEnabled=true
    Module.Param1WriteEnabled=true
    Module.Param2WriteEnabled=true
    Module.Param3WriteEnabled=true
   )
   &CameraOffset(Camera Offset Amount=-300)
   /SetVariables(Particles.VisibilityTag=&ModuloInt(A=&ReturnParticleIDAttributes(), B=2)) {enabled=false, isSetParameters=true}
   &Color(Color=&LerpLinearColors(
     StartColor=<rgba 1 1 1 1>
     EndColor=&LerpLinearColors(
      StartColor=<rgba 1 1 1 1>
      EndColor=&RandomRangeLinearColor(Minimum=<rgba 0.1039 0.7373 1 1>, Maximum=<rgba 0.1514 0.1329 1 1>)
      LerpFactor=0.5
     )
     LerpFactor=&MakeFloatFromInt(Int=&ModuloInt(A=&ReturnParticleIDAttributes(), B=2))
    ))
   &InitialMeshOrientation(Mesh Axis To Orient=<0 1 0>, Mesh Orientation Mode=::NewEnumerator2, Use Rotation Vector=false)
  ]
  particleUpdateScript [
   &SolveForcesAndVelocity(Clamp Velocity=false, Limit Acceleration=false)
   &ParticleState()
   &ScaleSpriteSize(Scale Sprite Size Mode=::NewEnumerator0, Uniform Scale Factor=2.567)
   &ScaleColor(
    Scale Mode=::NewEnumerator1
    Scale RGBA=&MakeVector4_FromLinearColor(LinearColor=&LerpLinearColors(
      StartColor=&ColorFromCurve(ColorCurve=DataInterface() {propertyValues {
         AlphaCurve {keys [<at 0 0>,<at 0.25 1>,<at 0.75 1>,<at 1 0>]}
         BlueCurve {keys [<at 0.5 1>]}
         GreenCurve {keys [<at 0.5 0.5916>]}
         RedCurve {keys [<at 0.5 0.5119>]}
        }})
      EndColor=&ColorFromCurve(ColorCurve=DataInterface() {propertyValues {
         AlphaCurve {keys [<at 0 0>,<at 0.25 1>,<at 0.75 1>,<at 1 0>]}
         BlueCurve {keys [<at 0 0.851>,<at 0.5 1>,<at 1 0.851>]}
         GreenCurve {keys [<at 0 0>,<at 0.5 0.302>,<at 1 0>]}
         RedCurve {keys [<at 0 1>,<at 0.5 0.0706>,<at 1 1>]}
        }})
      LerpFactor=&MakeFloatFromInt(Int=&ModuloInt(A=&ReturnParticleIDAttributes(), B=2))
     ))
    ScaleRGBA=true
   ) {module=ScaleColor001}
   &ScaleColor(
    Scale Alpha=&Multiply_Float(
     A=1
     B=&Waveform(
      Waveform Count=::NewEnumerator1
      Global Input Position=&RandomRangeFloat(Minimum=0, Maximum=1000)
      "Waveform [1]"=::NewEnumerator0
      "[1] Frequency"=1
      "[1] Amplitude Min/Max"=<0.4453 1.3014>
      "[2] Average"=false
      "[2] Input Position"=0
      "[2] Phase"=$Emitter.Age
      "[2] Phase Scale"=1
      "[2] Compound Frequency"=2
      "[2] Frequency"=2
      "[2] Amplitude Min/Max"=<0.5749 1.6256>
      "[2] Scale"=1
      "[2] Bias"=0
      "[2] Exponent"=1
      "[2] Pulse Duration"=0.1
      "[2] Clip Triangle Min/Max"=<0 0>
      "[2] Random Seed"=0
     )
    )
    Scale RGB=&Multiply_VectorByFloat(Vector=<1 1 1>, Float=3)
    ScaleA=true
    ScaleRGB=true
   )
   &ScaleColor(Scale Alpha=0.5, ScaleA=true, ScaleRGB=true) {module=ScaleColor002}
   &AlignSpriteToMeshOrientation(SetAlignment=true, SetFacing=true)
   &ScaleColor(
    Scale Alpha=&FloatFromCurve(
     FloatCurve=DataInterface() {propertyValues {Curve {keys [<at 0 0>, {arriveTangent=5.7107, interpMode=RCIM_Cubic, leaveTangent=5.7107, tangentMode=RCTM_User, time=1, value=1}]}}}
     CurveIndex=$System.DaggerCharge
    )
    Scale RGB=&VectorFromFloat(Value=1.25)
    ScaleA=true
    ScaleRGB=true
   ) {module=ScaleColor003}
  ]
  rendererClasses [&NiagaraSpriteRendererProperties]
  renderers [
   {rendererClass=&NiagaraSpriteRendererProperties}
   {rendererClass=&NiagaraSpriteRendererProperties, rendererIndex=1}
  ]
 }
 glow001 {
  emitterUpdateScript [
   &EmitterState()
   &SpawnBurst_Instantaneous(Use Spawn Probability=false, UseLoopCountLimit=true)
  ]
  particleSpawnScript [
   &InitializeParticle(
    Color=<rgba 0.3 1.5649 3 1>
    Lifetime=5
    Mesh Scale Mode=::NewEnumerator4
    Position Mode=::NewEnumerator1
    UsePositionOffset=false
    Write Material Random=true
   )
   &InitialMeshOrientation(Mesh Orientation Mode=::NewEnumerator2, Rotation=<0.25 0 0>, Use Rotation Vector=true)
   &CameraOffset(Camera Offset Amount=0.35)
  ]
  particleUpdateScript [
   &ParticleState(Kill Particles When Lifetime Has Elapsed=false, Let Infinitely Lived Particles Die When Emitter Deactivates=true)
   &ScaleMeshSize(Scale Factor=&Multiply_Vector(A=&VectorFromFloat(Value=1), B=<1 1 1>))
   &ScaleColor(
    Scale Alpha=&Multiply_Float(
     A=25
     B=&FloatFromCurve(
      FloatCurve=DataInterface() {propertyValues {Curve {keys [
          {
           interpMode=RCIM_Cubic
           leaveTangent=0.1701
           leaveTangentWeight=0.659
           tangentMode=RCTM_Break
           tangentWeightMode=RCTWM_WeightedBoth
           time=0
           value=0
          }
          {
           arriveTangent=4.523
           arriveTangentWeight=0.8247
           interpMode=RCIM_Cubic
           leaveTangent=2.2495
           tangentMode=RCTM_Break
           tangentWeightMode=RCTWM_WeightedBoth
           time=1
           value=1
          }
         ]}}}
      CurveIndex=$System.DaggerCharge
      Scale Curve=3
     )
    )
    ScaleA=true
    ScaleRGB=true
   )
  ]
  rendererClasses [&NiagaraMeshRendererProperties]
  renderers [{rendererClass=&NiagaraMeshRendererProperties, rendererIndex=0}]
 }
}
summary {
 systemName=NS_GrillMoonDagger
 emitters [
  {emitterName=ble_putrple_glow, rendererClasses [&NiagaraSpriteRendererProperties]}
  {emitterName=RainbowCorona003, rendererClasses [&NiagaraMeshRendererProperties]}
  {emitterName=Spiral_AirFlow, rendererClasses [&NiagaraMeshRendererProperties]}
  {emitterName=glow001, rendererClasses [&NiagaraMeshRendererProperties]}
  {emitterName=Sparklies, simTarget=GPUComputeSim, rendererClasses [&NiagaraSpriteRendererProperties]}
 ]
 userVariables [
  User.CoronaFlareScale {type=NiagaraFloat, defaultValue=1, description=""}
  User.DaggerCharge0to1 {type=NiagaraFloat, defaultValue=0.05, description=""}
  User.UseCombinedFX {type=NiagaraFloat, defaultValue=0, description=""}
 ]
}
