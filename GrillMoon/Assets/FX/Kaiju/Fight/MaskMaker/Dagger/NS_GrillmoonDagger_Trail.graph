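# Niagara weapon trail for the MaskMaker dagger in the GrillMoon Kaiju fight FX, drawn by one emitter with three ribbon renderers.
# Custom AnimTrail modules spawn, initialize and update short-lived ribbon particles from skeletal mesh bone or socket indices.
# Active update modules fade color and alpha over life and write dynamic material parameters, one varying by spawn group.
# User.UseFinalSlash scales separate blue FinalSlash color and width attributes, zeroing them at 0.
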NiagaraSystem /GrillMoon/Assets/FX/Kaiju/Fight/MaskMaker/Dagger/NS_GrillmoonDagger_Trail
dependencies {
 usedDataInterfaces [&NiagaraDataInterfaceCurve, &NiagaraDataInterfaceColorCurve, &NiagaraDataInterfaceArrayFloat]
 usedDynamicInputs []
 usedModules [&SystemState]
}
emitters {Trail {
  emitterUpdateScript [
   &EmitterState(Life Cycle Mode=::NewEnumerator1, Loop Behavior=::NewEnumerator1, Loop Duration=0.5, UseLoopDelay=true)
   &FN_SpawnAnimTrail_v2()
  ]
  particleSpawnScript [
   /Game/Effects/Niagara/Modules/AnimTrail/FN_InitializeAnimTrail/FN_InitializeAnimTrailBoneIndecies(Second Bone or Socket Index=1)
   /Game/Effects/Niagara/Modules/AnimTrail/FN_InitializeAnimTrail/FN_InitializeAnimTrail(AnimTrailWidth=1, Color=<rgba 1 1 1 3>, Lifetime=0.125)
  ]
  particleUpdateScript [
   &FN_UpdateAnimTrail_v2()
   &ScaleColor(
    Scale Alpha=&Multiply_Float(
     A=&FloatFromCurve(FloatCurve=DataInterface() {propertyValues {Curve {keys [
          <at 0 0.2 RCIM_Cubic>
          {arriveTangent=-0.0116, interpMode=RCIM_Cubic, leaveTangent=-0.4376, tangentMode=RCTM_Break, time=0.3, value=1}
          {arriveTangent=-1.6754, interpMode=RCIM_Cubic, leaveTangent=-1.6754, tangentMode=RCTM_User, time=0.6462, value=0.2255}
          {arriveTangent=-0.0464, interpMode=RCIM_Cubic, tangentMode=RCTM_Break, time=1, value=0}
         ]}}})
     B=&FloatFromCurve(
      FloatCurve=DataInterface() {propertyValues {Curve {keys [<at 0 1>,<at 0.1 1>,<at 0.5 0>,<at 1 0>]}}}
      CurveIndex=$Emitter.Age
      Scale Curve=&Multiply_Float(A=&FNTimeOfDayScalar(Night Scale=0.6, Morning Scale=1, Evening Scale=1), B=1)
     )
    )
    ScaleA=true
    ScaleRGB=true
   ) {enabled=false, module=ScaleColor002}
   /SetVariables(Particles.Color002=&MakeLinearColorFromVectorAndFloat(
     "Vector (RGB)"=&MakeVectorFromLinearColorRGB(Color=&ColorFromCurve(ColorCurve=DataInterface() {propertyValues {
         AlphaCurve {keys []}
         BlueCurve {keys [<at 0 0.4 RCIM_Cubic>]}
         GreenCurve {keys [<at 0 0.0246 RCIM_Cubic>]}
         RedCurve {keys [<at 0 0 RCIM_Cubic>]}
        }}))
     "Float (Alpha)"=&Multiply_Float(
      A=&FloatFromCurve(FloatCurve=DataInterface() {propertyValues {Curve {keys [
           <at 0 0 RCIM_Cubic>
           {arriveTangent=0.0207, interpMode=RCIM_Cubic, leaveTangent=0.0467, tangentMode=RCTM_Break, time=0.6, value=1}
           {arriveTangent=-0.0464, interpMode=RCIM_Cubic, tangentMode=RCTM_Break, time=1, value=0}
          ]}}})
      B=&FloatFromCurve(FloatCurve=DataInterface() {propertyValues {Curve {keys [
           <at 0.2 0 RCIM_Cubic>
           {arriveTangent=-0.7088, interpMode=RCIM_Cubic, leaveTangent=-0.7088, tangentMode=RCTM_User, time=0.5, value=1}
           {arriveTangent=-1.9114, interpMode=RCIM_Cubic, leaveTangent=-1.9114, tangentMode=RCTM_User, time=0.8, value=0}
          ]}}})
     )
    )) {enabled=false, isSetParameters=true}
   &DynamicMaterialParameters(
    Index 0 Param 1=&FloatFromCurve(FloatCurve=DataInterface() {propertyValues {Curve {keys [
         {arriveTangent=0.0174, interpMode=RCIM_Cubic, leaveTangent=0.0174, tangentMode=RCTM_User, time=0.5, value=0}
         {arriveTangent=-0.5399, interpMode=RCIM_Cubic, leaveTangent=-0.5399, tangentMode=RCTM_User, time=1, value=-0.5}
        ]}}})
    Module.Index0Param0Write=true
    Module.Index0Param1Write=true
    Module.Index0Param2Write=true
    Module.Index0Param3Write=true
   ) {enabled=false}
   &DynamicMaterialParameters(
    Index 0 Param 1=&SelectFloatFromArray(Direct Array Index=&SpawnGroup(), Float Selection Array=DataInterface() {propertyValues {FloatData [0,0.1]}})
    Index 0 Param 3=1
    Index 0 Param 4=1
    Module.Param0WriteEnabled=true
    Module.Param1WriteEnabled=false
    Module.Param2WriteEnabled=true
    Module.Param3WriteEnabled=true
   ) {module=DynamicMaterialParameters001}
   &ScaleColor(
    Scale Alpha=&FloatFromCurve(FloatCurve=DataInterface() {propertyValues {Curve {keys [<at 0 0>,<at 0.05 1>,<at 1 0>]}}}, Scale Curve=0.5)
    Scale RGB=&VectorFromFloat(Value=0.5)
    ScaleA=true
    ScaleRGB=true
   )
   /SetVariables(
    Particles.FinalSlash Color:LinearColor=&MultiplyLinearColor(
     Linear Color=&ColorFromCurve(ColorCurve=DataInterface() {propertyValues {
        AlphaCurve {keys [<at 0.6341 1 RCIM_Cubic>,<at 1 0 RCIM_Cubic>]}
        BlueCurve {keys [<at 0 1 RCIM_Cubic>,<at 1 1 RCIM_Cubic>]}
        GreenCurve {keys [<at 0 0.4488 RCIM_Cubic>,<at 1 0.4488 RCIM_Cubic>]}
        RedCurve {keys [<at 0 0 RCIM_Cubic>,<at 1 0 RCIM_Cubic>]}
       }})
     Scale Factor=&MakeFloat4FromVectorAndFloat(
      "Vector (RGB)"=&VectorFromFloat(Value=&Lerp_Float(A=0, B=5, Alpha=$User.UseFinalSlash))
      "Float (Alpha)"=&Lerp_Float(A=0, B=0.5, Alpha=$User.UseFinalSlash)
     )
    )
    Particles.Finalslash width:NiagaraFloat=&Multiply_Float(A=$Particles.RibbonWidth, B=&Lerp_Float(A=0, B=1.2, Alpha=$User.UseFinalSlash))
   ) {isSetParameters=true}
  ]
  rendererClasses [&NiagaraRibbonRendererProperties]
  renderers [
   {rendererClass=&NiagaraRibbonRendererProperties}
   {rendererClass=&NiagaraRibbonRendererProperties, rendererIndex=1}
   {rendererClass=&NiagaraRibbonRendererProperties, rendererIndex=2}
  ]
 }}
summary {
 systemName=NS_GrillmoonDagger_Trail
 emitters [{emitterName=Trail, rendererClasses [&NiagaraRibbonRendererProperties]}]
 userVariables [
  User.ParaGlideLeanAlpha {type=NiagaraFloat, defaultValue=0.5, description=""}
  User.RibbonIDOffset {type=NiagaraInt32}
  User.SkeletalMesh {
   type=NiagaraDataInterfaceSkeletalMesh
   defaultValue {
    struct=NiagaraExt_VariableValue_DataInterface
    value {dataInterface=&self:NiagaraDataInterfaceSkeletalMesh_0}
   }
  }
  User.UseFinalSlash {type=NiagaraFloat, defaultValue=1, description=""}
 ]
}
